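# Activatable modal dialog for Minifigs info, wrapping a single UIKit internal dialog widget styled with background, image, rich text and a bottom button group.
# PreConstruct branches on IsDesignTime to feed a constructed DialogVM_MinifigsInfo into the dialog as preview data; this graph sets no runtime data.
# On activation it clears and rebinds the dialog's button-click dispatcher to CloseModal, which deactivates the widget whichever button is pressed.
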WidgetBlueprint /FortniteCoreUI/MultiProduct/WBP_MinifigsInfo_Dialog
graphs [EventGraph {nodes [
   n1 = CloseModal!() -> (OutputDelegate, then, ButtonIndex:int) {is=title K2Node_CustomEvent}
   > DeactivateWidget()
   n2 = Event PreConstruct()
   > Branch(Condition=@n2.IsDesignTime)
   > n3 = Construct Object from Class(Class=/FortniteCoreUI/MultiProduct/Data/DialogVM_MinifigsInfo.DialogVM_MinifigsInfo_C) -> (ReturnValue:Dialog VM Minifigs Info Object Reference)
   > SetDialogData!(
    execute
    self:WBP UIKit Dialog Internal Object Reference=Get Dialog_Base() -> (Dialog_Base:WBP UIKit Dialog Internal Object Reference)
    Data:UI Kit Dialog Viewmodel Object Reference=@n3
   ) -> (then)
   Event On Activated()
   > Unbind all Events from Dialog Button Clicked Dispatcher!(execute:Exec, self:WBP UIKit Dialog Internal Object Reference=Get Dialog_Base()) -> (then:Exec) {is=title K2Node_ClearDelegate}
   > Bind Event to Dialog Button Clicked Dispatcher(self:WBP UIKit Dialog Internal Object Reference=Get Dialog_Base(), Delegate=@n1) {is=title K2Node_AddDelegate}
  ]}]
variables []
widgets {
 info {
  inheritedWidgetCount=0
  parentClass=&CommonActivatableWidget
  rootWidgetClass=/FortUILibrary/Systems/Dialog/WBP_UIKit_Dialog_Internal.WBP_UIKit_Dialog_Internal_C
  widgetCount=1
 }
 widgets [Dialog_Base {
   widgetClassPath=/FortUILibrary/Systems/Dialog/WBP_UIKit_Dialog_Internal
   bIsVariable=true
   widget=Dialog_Base
   properties {
    useImage=true
    useText=Common Rich Text
    backgroundProperties {
     material=/FortUILibrary/UIKit/Dialogs/MaterialInstances/MI_UI_DialogBackground
     image Size=<32 32>
     placement {
      horizontal Alignment=HAlign_Fill
      vertical Alignment=VAlign_Fill
      grid {column=0, column Span=2, layer=1, row=0, row Span=0, nudge=<0 0>}
      padding=<ltrb 0 0 0 0>
     }
    }
    buttons {
     each Button Horizontal Fill=false
     each Button Min Width=200
     each Button Size=Medium
     focuse Button=0
     space Between Buttons=30
     cTA Buttons Types [Regular,Quiet]
     group Placement {
      horizontal Alignment=HAlign_Fill
      vertical Alignment=VAlign_Bottom
      grid {column=1, column Span=0, layer=0, row=2, row Span=0, nudge=<0 0>}
      padding=<ltrb 80 64 80 48>
     }
    }
    dialogSIze {
     height Override {enable=false, value=800}
     max Aspect Ratio {enable=false, value=0}
     max Desired Height {enable=true, value=800}
     max Desired Width {enable=false, value=0}
     min Aspect Ratio {enable=false, value=0}
     min Desired Height {enable=false, value=0}
     min Desired Width {enable=false, value=0}
     width Override {enable=true, value=1664}
    }
    imageProperties {
     material=/FortUILibrary/Systems/Dialog/MaterialInstances/MI_UI_DialogImage
     image Size=<512 512>
     placement {
      horizontal Alignment=HAlign_Fill
      vertical Alignment=VAlign_Fill
      grid {column=1, column Span=0, layer=20, row=0, row Span=1, nudge=<0 0>}
      padding=<ltrb 0 0 0 0>
     }
    }
    textProperties {
     auto Wrap Text=true
     justification=Left
     line Height Percentage=1.2
     mobile Font Size Multiplier=1.3
     transform Policy=None
     wrap Text At=0
     color {colorUseRule=UseColor_Specified, specifiedColor=<rgba 1 1 1 0.75>}
     font {
      bForceMonospaced=false
      bMaterialIsStencil=false
      fontMaterial=null
      fontObject=&HeadingNow
      letterSpacing=0
      monospacedWidth=1
      size=15.75
      skewAmount=0
      typefaceFontName=74 Regular
      outlineSettings {
       bApplyOutlineToDropShadows=false
       bMiteredCorners=false
       bSeparateFillAlpha=false
       outlineMaterial=null
       outlineSize=0
       outlineColor=<rgba 0 0 0 1>
      }
     }
     placement {
      horizontal Alignment=HAlign_Fill
      vertical Alignment=VAlign_Top
      grid {column=1, column Span=0, layer=0, row=1, row Span=0, nudge=<0 0>}
      padding=<ltrb 80 0 80 0>
     }
    }
   }
  }]
}
