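# Season pass level widget showing the level number inside a progress image, beside an "XP to next LVL" readout on a dark backplate.
# SetCurrentXP and SetXPToNextLevel each take one value, read the other from FortPassLevelViewModel, and pass both to SetLevelProgressPercent.
# That function writes CurrentXP / (CurrentXP + XPToNextLevel) to the Progress parameter of LevelProgress's dynamic material.
# The graphs contain no logic that sets the text blocks.
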WidgetBlueprint /FortniteCoreUI/MultiProduct/SeasonPass_Level
graphs [
 EventGraph {links [], nodes []}
 SetCurrentXP {nodes [
   n1 = SetCurrentXP!() -> (then, CurrentXP:int) {is=title K2Node_FunctionEntry}
   > SetLevelProgressPercent!(
    execute
    self
    CurrentXP:int=@n1
    XPToNextLevel:int=Get XpToNextLevel!(self:Pass Level View Model Object Reference=Get FortPassLevelViewModel() -> (FortPassLevelViewModel:Pass Level View Model Object Reference)) -> (XpToNextLevel:int=0)
   ) -> (then)
  ]}
 SetLevelProgressPercent {nodes [
   n1 = SetLevelProgressPercent!() -> (then, CurrentXP:int, XPToNextLevel:int) {is=title K2Node_FunctionEntry}
   > n2 = GetDynamicMaterial(self=Get LevelProgress() -> (LevelProgress:Image Object Reference))
   > SetScalarParameterValue(self=@n2, ParameterName=Progress, Value=float / float(A:int=@n1, B:int=int + int(A=@n1.XPToNextLevel, B=@n1)))
  ]}
 SetXPToNextLevel {nodes [
   n1 = SetXPToNextLevel!() -> (then, XPToNextLevel:int) {is=title K2Node_FunctionEntry}
   > SetLevelProgressPercent(CurrentXP=Get CurrentXP!(self:Pass Level View Model Object Reference=Get FortPassLevelViewModel()) -> (CurrentXP:int=0), XPToNextLevel=@n1) {is=title K2Node_CallFunction}
  ]}
]
variables []
widgets {
 info {inheritedWidgetCount=0, parentClass=&UserWidget, rootWidgetClass=&Overlay, widgetCount=15}
 widgets [
  Overlay_21 {widgetClassPath=&Overlay, widget=Overlay_21}
  WBP_UIKit_Backplate {
   widgetClassPath=&WBP_UIKit_Backplate
   bIsVariable=true
   parent=Overlay_21
   widget=WBP_UIKit_Backplate
   properties {brightness=Dark, cornerRadius=Interactive Component}
   slot {class=OverlaySlot, properties {horizontalAlignment=HAlign_Fill, verticalAlignment=VAlign_Fill}}
  }
  HorizontalBox_31 {
   widgetClassPath=&HorizontalBox
   parent=Overlay_21
   widget=HorizontalBox_31
   slot {class=OverlaySlot, properties {horizontalAlignment=HAlign_Right, verticalAlignment=VAlign_Fill, padding=<ltrb 0 -16 -16 -16>}}
  }
  SizeBox_0 {
   widgetClassPath=&SizeBox
   parent=HorizontalBox_31
   widget=SizeBox_0
   properties {bOverride_MinDesiredWidth=true, minDesiredWidth=145}
   slot {class=HorizontalBoxSlot, properties {horizontalAlignment=HAlign_Right, verticalAlignment=VAlign_Center, padding=<ltrb 32 0 8 0>}}
  }
  VerticalBox_LevelInfo {widgetClassPath=&VerticalBox, bIsVariable=true, parent=SizeBox_0, widget=VerticalBox_LevelInfo, slot {class=SizeBoxSlot}}
  HorizontalBox_51 {
   widgetClassPath=&HorizontalBox
   parent=VerticalBox_LevelInfo
   widget=HorizontalBox_51
   slot {class=VerticalBoxSlot, properties {verticalAlignment=VAlign_Center, size {sizeRule=Fill, value=1}}}
  }
  Text_XpToNextLevel {
   widgetClassPath=&CommonTextBlock
   bIsVariable=true
   parent=HorizontalBox_51
   widget=Text_XpToNextLevel
   properties {
    bIsScrollingEnabled=false
    mobileFontSizeMultiplier=1.25
    text=XXXXX
    textTransformPolicy=ToUpper
    font {
     bForceMonospaced=false
     bMaterialIsStencil=false
     fontMaterial=null
     fontObject=&HeadingNow
     letterSpacing=0
     monospacedWidth=1
     size=22.5
     skewAmount=0
     typefaceFontName=75 Medium Italic
     outlineSettings {
      bApplyOutlineToDropShadows=false
      bMiteredCorners=false
      bSeparateFillAlpha=false
      outlineMaterial=null
      outlineSize=0
      outlineColor=<rgba 0 0 0 1>
     }
    }
    shadowOffset=<0 0>
    shapedTextOptions {bOverride_TextFlowDirection=true, bOverride_TextShapingMethod=false, textFlowDirection=Culture, textShapingMethod=Auto}
   }
   slot {class=HorizontalBoxSlot, properties {padding=<ltrb 0 0 8 0>}}
  }
  Text_Level_3 {
   widgetClassPath=&CommonTextBlock
   parent=HorizontalBox_51
   widget=Text_Level_3
   properties {
    bIsScrollingEnabled=false
    mobileFontSizeMultiplier=1.25
    text=XP
    textTransformPolicy=ToUpper
    font {
     bForceMonospaced=false
     bMaterialIsStencil=false
     fontMaterial=null
     fontObject=&HeadingNow
     letterSpacing=0
     monospacedWidth=1
     size=22.5
     skewAmount=0
     typefaceFontName=75 Medium Italic
     outlineSettings {
      bApplyOutlineToDropShadows=false
      bMiteredCorners=false
      bSeparateFillAlpha=false
      outlineMaterial=null
      outlineSize=0
      outlineColor=<rgba 0 0 0 1>
     }
    }
    shadowOffset=<0 0>
    shapedTextOptions {bOverride_TextFlowDirection=true, bOverride_TextShapingMethod=false, textFlowDirection=Culture, textShapingMethod=Auto}
   }
   slot {class=HorizontalBoxSlot}
  }
  Overlay_0 {
   widgetClassPath=&Overlay
   parent=VerticalBox_LevelInfo
   widget=Overlay_0
   slot {class=VerticalBoxSlot, properties {verticalAlignment=VAlign_Center, size {sizeRule=Fill, value=1}}}
  }
  Text_Level_2 {
   widgetClassPath=&CommonTextBlock
   parent=Overlay_0
   widget=Text_Level_2
   properties {
    bIsScrollingEnabled=false
    mobileFontSizeMultiplier=1.25
    text=to next LVL
    textTransformPolicy=ToUpper
    colorAndOpacity {colorUseRule=UseColor_Specified, specifiedColor=<rgba 0.5776 0.5776 0.5776 0.8>}
    font {
     bForceMonospaced=false
     bMaterialIsStencil=false
     fontMaterial=null
     fontObject=&HeadingNow
     letterSpacing=10
     monospacedWidth=1
     size=14.25
     skewAmount=0
     typefaceFontName=86 Bold
     outlineSettings {
      bApplyOutlineToDropShadows=false
      bMiteredCorners=false
      bSeparateFillAlpha=false
      outlineMaterial=null
      outlineSize=0
      outlineColor=<rgba 0 0 0 1>
     }
    }
    shadowOffset=<0 0>
    shapedTextOptions {bOverride_TextFlowDirection=true, bOverride_TextShapingMethod=false, textFlowDirection=Culture, textShapingMethod=Auto}
   }
   slot {class=OverlaySlot}
  }
  GridPanel_LevelBox {
   widgetClassPath=&GridPanel
   parent=HorizontalBox_31
   widget=GridPanel_LevelBox
   properties {columnFill [1], rowFill [1]}
   slot {class=HorizontalBoxSlot, properties {horizontalAlignment=HAlign_Right, padding=<ltrb -24 0 0 0>}}
  }
  LevelProgress {
   widgetClassPath=&Image
   bIsVariable=true
   parent=GridPanel_LevelBox
   widget=LevelProgress
   properties {
    visibility=SelfHitTestInvisible
    brush {
     bIsDynamicallyLoaded=false
     drawAs=Image
     imageType=FullColor
     mirroring=NoMirror
     resourceName=None
     resourceObject=/Game/UI/Frontend/Lobby/Habanero/Habanero_Materials/MaterialInstances/ProgressBars/MI_UI_Habanero_ProgressBar_PassLevel
     tiling=NoTile
     imageSize=<144 144>
     margin=<ltrb 0 0 0 0>
     outlineSettings {
      bUseBrushTransparency=false
      roundingType=HalfHeightRadius
      width=0
      color {colorUseRule=UseColor_Specified, specifiedColor=<rgba 0 0 0 0>}
      cornerRadii=<0 0 0 0>
     }
     tintColor {colorUseRule=UseColor_Specified, specifiedColor=<rgba 1 1 1 1>}
     uVRegion {bIsValid=false, max=<0 0>, min=<0 0>}
    }
   }
   slot {class=GridSlot}
  }
  VerticalBox_Level {
   widgetClassPath=&VerticalBox
   parent=GridPanel_LevelBox
   widget=VerticalBox_Level
   slot {class=GridSlot, properties {horizontalAlignment=HAlign_Center, layer=1, verticalAlignment=VAlign_Center}}
  }
  Text_2 {
   widgetClassPath=&CommonTextBlock
   parent=VerticalBox_Level
   widget=Text_2
   properties {
    bIsScrollingEnabled=false
    text=LVL
    textTransformPolicy=ToUpper
    colorAndOpacity {colorUseRule=UseColor_Specified, specifiedColor=<rgba 0.5776 0.5776 0.5776 0.8>}
    font {
     bForceMonospaced=false
     bMaterialIsStencil=false
     fontMaterial=null
     fontObject=&HeadingNow
     letterSpacing=0
     monospacedWidth=1
     size=15.75
     skewAmount=0
     typefaceFontName=75 Medium
     outlineSettings {
      bApplyOutlineToDropShadows=false
      bMiteredCorners=false
      bSeparateFillAlpha=false
      outlineMaterial=null
      outlineSize=0
      outlineColor=<rgba 0 0 0 1>
     }
    }
    shadowOffset=<0 0>
    shapedTextOptions {bOverride_TextFlowDirection=true, bOverride_TextShapingMethod=false, textFlowDirection=Culture, textShapingMethod=Auto}
   }
   slot {class=VerticalBoxSlot, properties {horizontalAlignment=HAlign_Center, verticalAlignment=VAlign_Top}}
  }
  Text_Level {
   widgetClassPath=&CommonTextBlock
   bIsVariable=true
   parent=VerticalBox_Level
   widget=Text_Level
   properties {
    bIsScrollingEnabled=false
    lineHeightPercentage=0.9
    mobileFontSizeMultiplier=1.25
    text=999
    font {
     bForceMonospaced=false
     bMaterialIsStencil=false
     fontMaterial=null
     fontObject=&HeadingNow
     letterSpacing=10
     monospacedWidth=1
     size=18
     skewAmount=0
     typefaceFontName=86 Bold
     outlineSettings {
      bApplyOutlineToDropShadows=false
      bMiteredCorners=false
      bSeparateFillAlpha=false
      outlineMaterial=null
      outlineSize=0
      outlineColor=<rgba 0 0 0 1>
     }
    }
    shadowOffset=<0 0>
    shapedTextOptions {bOverride_TextFlowDirection=true, bOverride_TextShapingMethod=false, textFlowDirection=Culture, textShapingMethod=Auto}
   }
   slot {
    class=VerticalBoxSlot
    properties {horizontalAlignment=HAlign_Center, verticalAlignment=VAlign_Bottom, padding=<ltrb 0 -4 0 0>, size {sizeRule=Fill, value=1}}
   }
  }
 ]
}
